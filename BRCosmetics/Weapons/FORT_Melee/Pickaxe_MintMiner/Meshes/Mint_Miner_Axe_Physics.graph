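# Physics asset for the Mint Miner two-handed pickaxe cosmetic, giving its bow and ribbons physics bodies for secondary motion.
# A single box covers the handle bone, with spheres on the bow and its two child pieces and thin boxes on two ribbons.
# The bow and both ribbons are constrained to the handle, and the bow pieces to the bow.
# Angular limits allow moderate swing, widest on the ribbons, with tight twist throughout.
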PhysicsAsset /BRCosmetics/Weapons/FORT_Melee/Pickaxe_MintMiner/Meshes/Mint_Miner_Axe_Physics
bodies {
 Melee_TwoHanded_Handle [{shapeType=Box, extentX=72.7122, extentY=4.8218, extentZ=106.9771, center=<1.1192 1.5477 11.7216>, rotation=<rot -2.917 -90 -0.3239>}]
 dyn_bow [{shapeType=Sphere, radius=12.3759, center=<7.6944 -0.1502 -0.3757>}]
 dyn_bow_a [{shapeType=Sphere, radius=4, center=<3 0 0>}]
 dyn_bow_b [{shapeType=Sphere, radius=4, center=<3 0 0>}]
 dyn_ribbon_a [{shapeType=Box, extentX=0.5689, extentY=3.0425, extentZ=6.5249, center=<3.8385 0 0>, rotation=<rot 89.2889 0 0>}]
 dyn_ribbon_b [{shapeType=Box, extentX=2.4573, extentY=3.0425, extentZ=6.2948, center=<4.4395 0 0>, rotation=<rot 90 0 0>}]
}
constraints [
 {bone1Name=dyn_bow, bone2Name=Melee_TwoHanded_Handle, swing1LimitDegrees=30, swing2LimitDegrees=30, twistLimitDegrees=5}
 {bone1Name=dyn_bow_a, bone2Name=dyn_bow, swing1LimitDegrees=25, swing2LimitDegrees=25, twistLimitDegrees=5}
 {bone1Name=dyn_bow_b, bone2Name=dyn_bow, swing1LimitDegrees=25, swing2LimitDegrees=25, twistLimitDegrees=5}
 {bone1Name=dyn_ribbon_a, bone2Name=Melee_TwoHanded_Handle, swing1LimitDegrees=45, swing2LimitDegrees=45, twistLimitDegrees=5}
 {bone1Name=dyn_ribbon_b, bone2Name=Melee_TwoHanded_Handle, swing1LimitDegrees=45, swing2LimitDegrees=45, twistLimitDegrees=5}
]
